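; Cliff static mesh actor for the Figment environment that lines its borders with four box volumes, named as grass cull volumes.
; Its construction script offsets each volume to an edge from the mesh's local bounds and stretches it along that edge, compensating for actor scale.
; BorderCullBoxWidth sets each child box's thickness, bInsideAlign shifts volumes by that box's extent, and bShowGrassCullVolumes toggles their visibility.
; The event graph holds only disabled stubs.
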(event EventBeginPlay
  (|CallParentFunction) ; This node is disabled and will not be called.\nDrag off pins to build functionality.
)

(event Collision|EventActorBeginOverlap
  (params (OtherActor :type Actor%20Object%20Reference))
  (|CallParentFunction OtherActor) ; This node is disabled and will not be called.\nDrag off pins to build functionality.
)

(event EventTick
  (params (DeltaSeconds :type Float%20%28single-precision%29))
  (|CallParentFunction DeltaSeconds) ; This node is disabled and will not be called.\nDrag off pins to build functionality.
)

(fn ConstructionScript
  (bind (_min _max) (Components|StaticMesh|GetLocalBounds (Variables|StaticMeshActor|GetStaticMeshComponent)))
  (bind (_min_2 _max_2) (Components|StaticMesh|GetLocalBounds (Variables|StaticMeshActor|GetStaticMeshComponent)))
  (bind (_min_3 _max_3) (Components|StaticMesh|GetLocalBounds (Variables|StaticMeshActor|GetStaticMeshComponent)))
  (bind (_min_4 _max_4) (Components|StaticMesh|GetLocalBounds (Variables|StaticMeshActor|GetStaticMeshComponent)))
  (bind (_min_5 _max_5) (Components|StaticMesh|GetLocalBounds (Variables|StaticMeshActor|GetStaticMeshComponent)))
  (bind _bordercullboxwidth (Variables|Default|GetBorderCullBoxWidth))
  (bind _returnvalue (Math|Vector|MakeVector _bordercullboxwidth 1.0 1.0))
  (bind (_min_6 _max_6) (Components|StaticMesh|GetLocalBounds (Variables|StaticMeshActor|GetStaticMeshComponent)))
  (bind _bordercullboxwidth_2 (Variables|Default|GetBorderCullBoxWidth))
  (bind _returnvalue_2 (Math|Vector|MakeVector _bordercullboxwidth_2 1.0 1.0))
  (bind _bordercullboxwidth_3 (Variables|Default|GetBorderCullBoxWidth))
  (bind _returnvalue_3 (Math|Vector|MakeVector 1.0 _bordercullboxwidth_3 1.0))
  (bind (_min_7 _max_7) (Components|StaticMesh|GetLocalBounds (Variables|StaticMeshActor|GetStaticMeshComponent)))
  (bind _bordercullboxwidth_4 (Variables|Default|GetBorderCullBoxWidth))
  (bind _returnvalue_4 (Math|Vector|MakeVector 1.0 _bordercullboxwidth_4 1.0))
  (bind (_min_8 _max_8) (Components|StaticMesh|GetLocalBounds (Variables|StaticMeshActor|GetStaticMeshComponent)))
  (bind _grasscullx0 (Variables|Default|GetGrassCullX0))
  (bind _grassculx1 (Variables|Default|GetGrassCulX1))
  (bind _grasscully2 (Variables|Default|GetGrassCullY2))
  (bind _grassculy1 (Variables|Default|GetGrassCulY1))
  (bind _bshowgrasscullvolumes (Variables|Default|GetbShowGrassCullVolumes))
  (|CallParentFunction)
  (Utilities|FlowControl|Sequence
    (:then_0
      (Utilities|FlowControl|Sequence
        (:then_1
          ; Comment
          (Variables|/Script/Engine.SceneComponent|SetbAbsoluteScale true _grasscullx0)
          (bind _asbox_collision (Utilities|Casting|CastToBoxComponent (Components|GetChildComponent _grasscullx0 0)))
          (Transformation|SetRelativeScale3D _asbox_collision _returnvalue)
          (Transformation|AddLocalOffset _grasscullx0 (+ (Math|Vector|MakeVector 0.0 (.y (* (Math|Vector|MakeVector (.x _max_5) (.y _min_5) 1.0) 0.5)) 1.0) (select (Variables|Default|GetbInsideAlign) (Math|Vector|MakeVector (.x (* (/ 1.0 (Transformation|GetActorScale3D)) (* (Components|Box|GetUnscaledBoxExtent _asbox_collision) _returnvalue)))) "0, 0, 0")) false false)
          (Transformation|SetWorldScale3D _grasscullx0 (Math|Vector|MakeVector 1.0 (.y (/ (* (* (- _max _min) 0.5) (Transformation|GetActorScale3D)) (* (* (Components|Box|GetUnscaledBoxExtent _asbox_collision) _returnvalue) "1.000000,-1.000000,1.000000"))) 1.0)))
        (:then_2
          ; Comment
          (Variables|/Script/Engine.SceneComponent|SetbAbsoluteScale true _grassculy1)
          (bind _asbox_collision_2 (Utilities|Casting|CastToBoxComponent (Components|GetChildComponent _grassculy1 0)))
          (Transformation|SetRelativeScale3D _asbox_collision_2 _returnvalue_3)
          (Transformation|AddLocalOffset _grassculy1 (+ (Math|Vector|MakeVector (* (.x (Math|Vector|MakeVector (.x _max_6) (.y _min_6) 1.0)) 0.5) (.y (Math|Vector|MakeVector (.x _max_6) (.y _min_6) 1.0))) (select (Variables|Default|GetbInsideAlign) (Math|Vector|MakeVector 0.0 (.y (* (/ 1.0 (Transformation|GetActorScale3D)) (* (Components|Box|GetUnscaledBoxExtent _asbox_collision_2) _returnvalue_3)))) "0, 0, 0")) false false)
          (Transformation|SetWorldScale3D _grassculy1 (Math|Vector|MakeVector (.x (/ (* (* (- _max_4 _min_4) 0.5) (Transformation|GetActorScale3D)) (* (Components|Box|GetUnscaledBoxExtent _asbox_collision_2) "1.000000,-1.000000,1.000000"))) 1.0 1.0)))
        (:then_3
          ; Comment
          (Variables|/Script/Engine.SceneComponent|SetbAbsoluteScale false _grassculx1)
          (bind _asbox_collision_3 (Utilities|Casting|CastToBoxComponent (Components|GetChildComponent _grassculx1 0)))
          (Transformation|SetRelativeScale3D _asbox_collision_3 _returnvalue_4)
          (Transformation|AddLocalOffset _grassculx1 (+ (Math|Vector|MakeVector (.x (* (Math|Vector|MakeVector (.x _max_7) (.y _min_7) 1.0) 0.5))) (select (Variables|Default|GetbInsideAlign) (Math|Vector|MakeVector 0.0 (.y (* (/ 1.0 (Transformation|GetActorScale3D)) (* (* (Components|Box|GetUnscaledBoxExtent _asbox_collision_3) _returnvalue_4) "1.000000,-1.000000,1.000000")))) "0, 0, 0")) false false)
          (Transformation|SetWorldScale3D _grassculx1 (Math|Vector|MakeVector (.x (/ (* (* (- _max_3 _min_3) 0.5) (Transformation|GetActorScale3D)) (* (* (Components|Box|GetUnscaledBoxExtent _asbox_collision_3) _returnvalue_4) "1.000000,-1.000000,1.000000"))) 1.0 1.0)))
        (:then_4
          ; Comment
          (Variables|/Script/Engine.SceneComponent|SetbAbsoluteScale true _grasscully2)
          (bind _asbox_collision_4 (Utilities|Casting|CastToBoxComponent (Components|GetChildComponent _grasscully2 0)))
          (Transformation|SetRelativeScale3D _asbox_collision_4 _returnvalue_2)
          (Transformation|AddLocalOffset _grasscully2 (- (Math|Vector|MakeVector (.x (Math|Vector|MakeVector (.x _max_8) (.y _min_8) 1.0)) (* (.y (Math|Vector|MakeVector (.x _max_8) (.y _min_8) 1.0)) 0.5)) (select (Variables|Default|GetbInsideAlign) (Math|Vector|MakeVector (.x (* (/ 1.0 (Transformation|GetActorScale3D)) (* (* (Components|Box|GetUnscaledBoxExtent _asbox_collision_4) _returnvalue_2) "1.000000,-1.000000,1.000000")))) "0, 0, 0")) false false)
          (Transformation|SetWorldScale3D _grasscully2 (Math|Vector|MakeVector 1.0 (.y (/ (* (* (- _max_2 _min_2) 0.5) (Transformation|GetActorScale3D)) (* (Components|Box|GetUnscaledBoxExtent _asbox_collision_4) _returnvalue_2))) 1.0)))))
    (:then_1
      (Rendering|SetVisibility (Variables|Default|GetGrassCulX1) _bshowgrasscullvolumes false))
    (:then_2
      (Rendering|SetVisibility (Variables|Default|GetGrassCullX0) _bshowgrasscullvolumes false))
    (:then_3
      (Rendering|SetVisibility (Variables|Default|GetGrassCulY1) _bshowgrasscullvolumes false))
    (:then_4
      (Rendering|SetVisibility (Variables|Default|GetGrassCullY2) _bshowgrasscullvolumes false))))
