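; Cliff static mesh actor Blueprint that sizes a box to the whole mesh footprint, apparently to cull grass around the cliff.
; The construction script gives the ChildActor component absolute scale and offsets it to the mesh's scaled local bounds center.
; It scales the child box's X and Y from scaled mesh half-extents, unscaled box extent and CullBoxMultiplier, keeping Z scale at 1.
; The event graph holds only disabled stubs.
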(event EventBeginPlay
  (|CallParentFunction) ; This node is disabled and will not be called.\nDrag off pins to build functionality.
)

(event Collision|EventActorBeginOverlap
  (params (OtherActor :type Actor%20Object%20Reference))
  (|CallParentFunction OtherActor) ; This node is disabled and will not be called.\nDrag off pins to build functionality.
)

(event EventTick
  (params (DeltaSeconds :type Float%20%28single-precision%29))
  (|CallParentFunction DeltaSeconds) ; This node is disabled and will not be called.\nDrag off pins to build functionality.
)

(fn ConstructionScript
  (bind (_min _max) (Components|StaticMesh|GetLocalBounds (Variables|StaticMeshActor|GetStaticMeshComponent)))
  (bind (_min_2 _max_2) (Components|StaticMesh|GetLocalBounds (Variables|StaticMeshActor|GetStaticMeshComponent)))
  (bind _childactor (Variables|Default|GetChildActor))
  (|CallParentFunction)
  (Variables|/Script/Engine.SceneComponent|SetbAbsoluteScale true _childactor)
  (bind _asbox_collision (Utilities|Casting|CastToBoxComponent (Components|GetChildComponent _childactor 0)))
  (Transformation|AddLocalOffset (Variables|Default|GetChildActor) (* (* (+ _min_2 _max_2) 0.5) (Transformation|GetActorScale3D)) false false)
  (Transformation|SetRelativeScale3D _asbox_collision (Math|Vector|MakeVector (.x (/ (* (* (- _max _min) 0.5) (Transformation|GetActorScale3D)) (/ (Components|Box|GetUnscaledBoxExtent _asbox_collision) (Variables|Default|GetCullBoxMultiplier)))) (.y (/ (* (* (- _max _min) 0.5) (Transformation|GetActorScale3D)) (/ (Components|Box|GetUnscaledBoxExtent _asbox_collision) (Variables|Default|GetCullBoxMultiplier)))) 1.0)))
